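; Agency courtyard brazier prop that applies Curie fire interactions to actors inside its sphere and suppresses native Curie FX.
; Once playlist data is ready, a hotfix scalable float toggles sphere collision; when enabled with Curie active, initial overlaps are processed.
; Authority-side overlaps begin and end fire interactions, storing handles per actor in a map.
; Death effects fade the fire audio; destruction ends all tracked interactions so nothing burns eternally.
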(event EventOnDeathPlayEffects
  (params (Damage :type Float%20%28single-precision%29) (DamageTags :type Gameplay%20Tag%20Container%20Structure%20%28by%20ref%29) (Momentum :type Vector :default "0, 0, 0") (HitInfo :type Hit%20Result%20Structure%20%28by%20ref%29) (InstigatedBy :type Fort%20Pawn%20Object%20Reference) (DamageCauser :type Actor%20Object%20Reference) (EffectContext :type Gameplay%20Effect%20Context%20Handle%20Structure))
  (bind _fire_audio_loop (Variables|Default|GetFire%20Audio%20Loop)
    (:then
      (Audio|Components|Audio|FadeOut _fire_audio_loop 0.25 0.0 "Linear" :!class /Script/Engine.AudioComponent))
    (:else)))

(event EventBeginPlay
  (bind _sphere (Variables|Default|GetSphere))
  ;[1 1 1] Hotfix to turn overlap off incase we need to kill this for perf.
  (Class|FortAsyncActionCurrentPlaylistReady|PlaylistDataReadyAsync
    (:OnReady
      (if (> (ScalableFloat|GetValueatLevel (Variables|Default|GetCollisionHotFixRow)) 0.0)
        (Collision|SetCollisionEnabled _sphere "QueryOnly")
        (if (Curie|IsCurieActive)
          (CallFunction|CheckInitialOverlappingActors))
        (else
          (Collision|SetCollisionEnabled _sphere "NoCollision"))))))

(event OnComponentBeginOverlap(Sphere)
  (params (OverlappedComponent :type Primitive%20Component%20Object%20Reference) (OtherActor :type Actor%20Object%20Reference) (OtherComp :type Primitive%20Component%20Object%20Reference) (OtherBodyIndex :type Integer) (bFromSweep :type Boolean) (SweepResult :type Hit%20Result%20Structure%20%28by%20ref%29))
  (Networking|SwitchHasAuthority
    (:Authority
      (if (Curie|IsCurieActive)
        (if (not (Utilities|Map|Contains (Variables|Default|GetInteractHandleMap) OtherActor))
          (bind _returnvalue (Curie|ExecuteBeginFireInteract self self OtherActor 1e+02 SweepResult))
          (Utilities|Map|Add (Variables|Default|GetInteractHandleMap) OtherActor _returnvalue))))))

(event OnComponentEndOverlap(Sphere)
  (params (OverlappedComponent :type Primitive%20Component%20Object%20Reference) (OtherActor :type Actor%20Object%20Reference) (OtherComp :type Primitive%20Component%20Object%20Reference) (OtherBodyIndex :type Integer))
  (bind _interacthandlemap (Variables|Default|GetInteractHandleMap))
  (bind (_value _returnvalue) (Utilities|Map|Find _interacthandlemap OtherActor))
  (Networking|SwitchHasAuthority
    (:Authority
      (if (Curie|IsCurieActive)
        (if _returnvalue
          (Curie|ExecuteEndFireInteract OtherActor _value)
          (Utilities|Map|Remove (Variables|Default|GetInteractHandleMap) OtherActor))))))

(event EventDestroyed
  (bind _interacthandlemap (Variables|Default|GetInteractHandleMap))
  (bind _interacthandlemap_2 (Variables|Default|GetInteractHandleMap))
  ;[1 1 1] Remove all fire created by this guy if destroyed.  Otherwise, people will burn eternally.
  (bind _keys (Utilities|Map|Keys _interacthandlemap_2))
  (for _array_element _keys
    (if (.ReturnValue (Utilities|Map|Find _interacthandlemap _array_element))
      (Curie|ExecuteEndFireInteract _array_element (Utilities|Map|Find _interacthandlemap _array_element))))
  (Utilities|Map|Clear _interacthandlemap_2))

(event Custom|CheckInitialOverlappingActors
  (Networking|SwitchHasAuthority
    (:Authority
      (for _array_element (Collision|GetOverlappingActors (Variables|Default|GetSphere) 0 :!class /Script/Engine.PrimitiveComponent)
        (if (not (Utilities|Map|Contains (Variables|Default|GetInteractHandleMap) _array_element))
          (bind _returnvalue (Curie|ExecuteBeginFireInteract self self _array_element 1e+02 (Collision|MakeHitResult false false 0.0 0.0 (Transformation|GetActorLocation _array_element) (Transformation|GetActorLocation _array_element) "0,0,1" "0,0,1" 0 _array_element)))
          (Utilities|Map|Add (Variables|Default|GetInteractHandleMap) _array_element _returnvalue))))))

(fn ShouldPlayNativeCurieFX
  (params (FXTag :type Gameplay%20Tag%20Structure))
  (return (Utilities|Struct|MakeFortNativeCurieFXResponse :GameplayCueResponse "IgnoreCue" :bShouldPlayGeneralVFX false :bShouldPlayAmbientAudio false :bShouldPlayGlow false :bShouldPlayFXAsAOE false)))

(fn ConstructionScript
  (|CallParentFunction))
