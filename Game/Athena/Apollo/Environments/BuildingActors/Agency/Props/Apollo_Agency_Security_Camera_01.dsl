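; Agency security camera prop that sweeps its Camera component back and forth through a configurable rotation cone.
; StartRotation eases the component to a yaw and pitch scaled by DirectionBias over RotDuration, then flips the bias and calls itself again.
; When not recently rendered, it stops sweeping and starts a randomized looping timer that polls render state, clearing itself and resuming rotation once visible.
; BeginPlay is empty here.
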(event EventBeginPlay)

(event Custom|StartRotation
  (if (Rendering|WasActorRecentlyRendered 0.2)
    (Components|MoveComponentTo (Variables|Default|GetCamera) (Variables|/Script/Engine.SceneComponent|GetRelativeLocation (Variables|Default|GetCamera)) (Math|Rotator|MakeRotator 0.0 (* (Variables|Components|GetRandomPitch) (Variables|Default|GetDirectionBias)) (* (Variables|Default|GetDirectionBias) (Variables|Components|GetRotConeInDegrees))) true true (Variables|Components|GetRotDuration) true)
    (Math|Float|NegateFloat (Variables|Default|GetDirectionBias)
      (:%20%20
        (CallFunction|StartRotation)))
    (else
      (bind _returnvalue (Utilities|Time|SetTimerbyEvent (delegate Custom|StartTickingIfRendered) (Math|Random|RandomFloatinRange 3 5) true))
      (Variables|Default|SetTimerHandle _returnvalue))))

(event Custom|StartTickingIfRendered
  (if (Rendering|WasActorRecentlyRendered 0.2)
    (Utilities|Time|ClearandInvalidateTimerbyHandle (Variables|Default|GetTimerHandle))
    (CallFunction|StartRotation)))

(fn ConstructionScript)
