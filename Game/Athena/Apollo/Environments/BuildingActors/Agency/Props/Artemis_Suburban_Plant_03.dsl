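; Blueprint for a suburban plant prop, Artemis_Suburban_Plant_03, among the Athena Apollo Agency environment building actors.
; Its construction script only calls the parent class's construction logic, adding no behavior of its own.
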(fn ConstructionScript
  (|CallParentFunction))
